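# Translucent unlit material in the Spire effects folder holding a thin-film iridescence graph, with none of its chains wired to a material output.
# A spectrum texture is sampled at a warped Fresnel value plus a static-switched offset, either fixed or animated by modulo time.
# The intensity-scaled colour is blended by a second Fresnel and multiplied.
# Several base inputs, including those of a separate roughness-scale clamp, are unconnected.
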Material /Game/Effects/Fort_Effects/GameplayPulginsAssets/SPire/M_Transparent
graph {
 properties {bAutomaticallySetUsageInEditor=false, blendMode=BLEND_Translucent, shadingModel=MSM_Unlit, usedShadingModels=Unlit}
 nodes [
  Clamp(None=Multiply(B=ScalarParameter() {defaultValue=1, group=ThinFilm, parameterName=ThinFilm_RoughnessScale}))
  n1 = StaticSwitchParameter(
   True=MaterialFunctionCall!(
    Time
    Divisor=ScalarParameter() {defaultValue=1, group=ThinFilm, parameterName=ThinFilm_Animated_TimeDivisor}
    Mod=ScalarParameter() {defaultValue=1, group=ThinFilm, parameterName=ThinFilm_Animated_TimeModulo}
   ) {materialFunction=/Game/Effects/Fort_Effects/Materials/_Functions/MF_ModuloTime}.Unclamped
   False=ScalarParameter() {group=ThinFilm, parameterName=ThinFilm_Offset}
  ) {group=ThinFilm, parameterName=AnimateThinFilm}
  n2 = Multiply(
   A=ScalarParameter() {defaultValue=1, group=ThinFilm, parameterName=ThinFilm_Scale}
   B=Fresnel(ExponentIn=ScalarParameter() {defaultValue=0.5, group=ThinFilm, parameterName=ThinFilm_Warp}) {baseReflectFraction=0}
  )
  n3 = Multiply(
   A=TextureSampleParameter2D(UVs=Add(A=@n1, B=@n2)) {group=ThinFilm, parameterName=Thin Film Texture, texture=/Game/Global/Textures/Default/Thinfilm/T_ThinFilm_Spectrum_COLOR}.RGB
   B=ScalarParameter() {defaultValue=0.5, group=ThinFilm, parameterName=ThinFilm_Intensity}
  )
  n4 = Fresnel(ExponentIn=ScalarParameter() {defaultValue=1.2164, group=ThinFilm, parameterName=ThinFilm_Exponent}) {baseReflectFraction=0}
  Multiply(A=LinearInterpolate(B=Add(B=@n3), Alpha=Multiply(B=@n4) {constA=1})) {constB=4}
 ]
}
